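# Animation Blueprint for the QuarterClaspZoom pickaxe that adds physical sway to its dyn_vbuck bones in top, mid and bottom groups.
# Two parallel chains run per-bone Anim Dynamics on the input pose, with angular springs, limits and custom gravity.
# bIsInFrontEnd selects the stiffer chain with simulation-space angular settings in the front end, otherwise the chain driven by scaled, clamped component linear motion.
# The event graph has no logic.
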AnimBlueprint /BRCosmetics/Weapons/FORT_Melee/Pickaxe_QuarterClaspZoom/Meshes/QuarterClaspZoom_Axe_AnimBP
target_skeleton=&/QuarterClaspZoom_Axe_Skeleton
graphs [
 AnimGraph {nodes [
   AnimGraphNode_SubInput_0 = Input Pose()
   n1 = Anim Dynamics - Bone: dyn_vbuck_top_A_01(ComponentPose=Local To Component(LocalPose=@AnimGraphNode_SubInput_0), Alpha=1, ExternalForce="0,0,0") {
    angularDampingOverride=0.97
    angularSpringConstant=300
    bAngularSpring=true
    bOverrideAngularDamping=true
    bUseGravityOverride=true
    boundBone {boneName=dyn_vbuck_top_A_01}
    componentAppliedLinearAccClamp=<700 500 350>
    componentLinearAccScale=<0.3 0.3 0.3>
    componentLinearVelScale=<0.4 0.4 0.4>
    gravityOverride {z=-700}
    physicsBodyDefinitions [{
      boundBone {boneName=dyn_vbuck_top_A_01}
      boxExtents=<5 5 5>
      constraintSetup {angularLimitsMax=<0 10 35>, angularLimitsMin=<0 -10 -5>, angularTarget=<1 0 0>}
      localJointOffset=<1 0 0>
     }]
   }
   n2 = Anim Dynamics - Bone: dyn_vbuck_top_A_02(ComponentPose=@n1, Alpha=1, ExternalForce="0,0,0") {
    angularDampingOverride=0.987
    angularSpringConstant=100
    bAngularSpring=true
    bOverrideAngularDamping=true
    bUseGravityOverride=true
    boundBone {boneName=dyn_vbuck_top_A_02}
    componentAppliedLinearAccClamp=<900 1000 350>
    componentLinearAccScale=<0.5 0.5 0.5>
    componentLinearVelScale=<0.6 0.6 0.5>
    gravityOverride {z=-980}
    physicsBodyDefinitions [{
      boundBone {boneName=dyn_vbuck_top_A_02}
      boxExtents=<10 5 5>
      constraintSetup {angularLimitsMax=<0 10 5>, angularLimitsMin=<0 -10 -50>, angularTarget=<1 0 0>}
      localJointOffset=<2 0 0>
     }]
   }
   n3 = Anim Dynamics - Bone: dyn_vbuck_top_B_01(ComponentPose=@n2, Alpha=1, ExternalForce="0,0,0") {
    angularDampingOverride=0.987
    angularSpringConstant=500
    bAngularSpring=true
    bOverrideAngularDamping=true
    bUseGravityOverride=true
    boundBone {boneName=dyn_vbuck_top_B_01}
    componentAppliedLinearAccClamp=<700 500 1000>
    componentLinearAccScale=<0.6 0.6 0.6>
    componentLinearVelScale=<0.7 0.7 0.7>
    gravityOverride {z=-800}
    physicsBodyDefinitions [{
      boundBone {boneName=dyn_vbuck_top_B_01}
      boxExtents=<10 5 10>
      constraintSetup {angularLimitsMax=<0 10 50>, angularLimitsMin=<0 -10 0>, angularTarget=<1 0 0>}
      localJointOffset=<1 0 0>
     }]
   }
   n4 = Anim Dynamics - Bone: dyn_vbuck_top_C_01(ComponentPose=@n3, Alpha=1, ExternalForce="0,0,0") {
    angularDampingOverride=0.987
    angularSpringConstant=500
    bAngularSpring=true
    bOverrideAngularDamping=true
    bUseGravityOverride=true
    boundBone {boneName=dyn_vbuck_top_C_01}
    componentAppliedLinearAccClamp=<500 500 700>
    componentLinearAccScale=<0.6 0.6 0.6>
    componentLinearVelScale=<0.7 0.7 0.7>
    gravityOverride {z=-800}
    physicsBodyDefinitions [{
      boundBone {boneName=dyn_vbuck_top_C_01}
      boxExtents=<10 5 10>
      constraintSetup {angularLimitsMax=<0 0 40>, angularLimitsMin=<0 0 0>, angularTarget=<1 0 0>}
      localJointOffset=<0.5 0 0>
     }]
   }
   n5 = Anim Dynamics - Bone: dyn_vbuck_top_C_02(ComponentPose=@n4, Alpha=1, ExternalForce="0,0,0") {
    angularDampingOverride=0.999
    angularSpringConstant=100
    bAngularSpring=true
    bOverrideAngularDamping=true
    bUseGravityOverride=true
    boundBone {boneName=dyn_vbuck_top_C_02}
    componentAppliedLinearAccClamp=<700 500 350>
    componentLinearAccScale=<0.6 0.6 0.6>
    componentLinearVelScale=<0.7 0.7 0.7>
    gravityOverride {z=-980}
    physicsBodyDefinitions [{
      boundBone {boneName=dyn_vbuck_top_C_02}
      boxExtents=<5 10 5>
      constraintSetup {angularLimitsMax=<0 10 0>, angularLimitsMin=<0 -10 -50>, angularTarget=<1 0 0>}
      localJointOffset=<1.5 0 0>
     }]
   }
   n6 = Anim Dynamics - Bone: dyn_vbuck_top_D_01(ComponentPose=@n5, Alpha=1, ExternalForce="0,0,0") {
    angularDampingOverride=0.9998
    angularSpringConstant=100
    bAngularSpring=true
    bOverrideAngularDamping=true
    bUseGravityOverride=true
    boundBone {boneName=dyn_vbuck_top_D_01}
    componentAppliedLinearAccClamp=<500 500 350>
    componentLinearAccScale=<0.6 0.6 0.6>
    componentLinearVelScale=<0.7 0.7 0.7>
    gravityOverride {z=-800}
    physicsBodyDefinitions [{
      boundBone {boneName=dyn_vbuck_top_D_01}
      boxExtents=<10 5 5>
      constraintSetup {angularLimitsMax=<0 10 7>, angularLimitsMin=<0 -10 -50>, angularTarget=<1 0 0>}
      localJointOffset=<1 0 0>
     }]
   }
   n7 = Anim Dynamics - Bone: dyn_vbuck_top_E_01(ComponentPose=@n6, Alpha=1, ExternalForce="0,0,0") {
    angularDampingOverride=0.999
    angularSpringConstant=100
    bAngularSpring=true
    bOverrideAngularDamping=true
    bUseGravityOverride=true
    boundBone {boneName=dyn_vbuck_top_E_01}
    componentAppliedLinearAccClamp=<700 500 350>
    componentLinearAccScale=<0.6 0.6 0.6>
    componentLinearVelScale=<0.7 0.7 0.7>
    gravityOverride {z=-980}
    physicsBodyDefinitions [{
      boundBone {boneName=dyn_vbuck_top_E_01}
      boxExtents=<10 5 5>
      constraintSetup {angularLimitsMax=<0 10 10>, angularLimitsMin=<0 -10 -50>, angularTarget=<1 0 0>}
      localJointOffset=<2 0 0>
     }]
   }
   n8 = Anim Dynamics - Bone: dyn_vbuck_mid_A_01(ComponentPose=@n7, Alpha=1, ExternalForce="0,0,0") {
    angularDampingOverride=0.988
    angularSpringConstant=100
    bAngularSpring=true
    bOverrideAngularDamping=true
    bUseGravityOverride=true
    boundBone {boneName=dyn_vbuck_mid_A_01}
    componentAppliedLinearAccClamp=<700 500 300>
    componentLinearAccScale=<0.6 0.6 0.6>
    componentLinearVelScale=<0.7 0.7 0.7>
    gravityOverride {z=-800}
    physicsBodyDefinitions [{
      boundBone {boneName=dyn_vbuck_mid_A_01}
      boxExtents=<5 5 10>
      constraintSetup {angularLimitsMax=<0 10 35>, angularLimitsMin=<0 -10 -15>, angularTarget=<1 0 0>}
      localJointOffset=<1 0 0>
     }]
   }
   n9 = Anim Dynamics - Bone: dyn_vbuck_mid_A_02(ComponentPose=@n8, Alpha=1, ExternalForce="0,0,0") {
    angularDampingOverride=0.987
    angularSpringConstant=100
    bAngularSpring=true
    bOverrideAngularDamping=true
    bUseGravityOverride=true
    boundBone {boneName=dyn_vbuck_mid_A_02}
    componentAppliedLinearAccClamp=<700 500 350>
    componentLinearAccScale=<0.6 0.6 0.6>
    componentLinearVelScale=<0.7 0.7 0.7>
    gravityOverride {z=-980}
    physicsBodyDefinitions [{
      boundBone {boneName=dyn_vbuck_mid_A_02}
      boxExtents=<10 5 5>
      constraintSetup {angularLimitsMax=<0 10 0>, angularLimitsMin=<0 -10 -50>, angularTarget=<1 0 0>}
      localJointOffset=<2 0 0>
     }]
   }
   n10 = Anim Dynamics - Bone: dyn_vbuck_mid_B_01(ComponentPose=@n9, Alpha=1, ExternalForce="0,0,0") {
    angularDampingOverride=0.997
    angularSpringConstant=100
    bAngularSpring=true
    bOverrideAngularDamping=true
    bUseGravityOverride=true
    boundBone {boneName=dyn_vbuck_mid_B_01}
    componentAppliedLinearAccClamp=<700 500 1000>
    componentLinearAccScale=<0.6 0.6 0.6>
    componentLinearVelScale=<0.7 0.7 0.7>
    gravityOverride {z=-800}
    physicsBodyDefinitions [{
      boundBone {boneName=dyn_vbuck_mid_B_01}
      boxExtents=<5 5 10>
      constraintSetup {angularLimitsMax=<0 10 35>, angularLimitsMin=<0 -10 0>, angularTarget=<1 0 0>}
      localJointOffset=<2 0 0>
     }]
   }
   n11 = Anim Dynamics - Bone: dyn_vbuck_mid_B_02(ComponentPose=@n10, Alpha=1, ExternalForce="0,0,0") {
    angularDampingOverride=0.999
    angularSpringConstant=100
    bAngularSpring=true
    bOverrideAngularDamping=true
    bUseGravityOverride=true
    boundBone {boneName=dyn_vbuck_mid_B_02}
    componentAppliedLinearAccClamp=<700 500 350>
    componentLinearAccScale=<0.6 0.6 0.6>
    componentLinearVelScale=<0.7 0.7 0.7>
    gravityOverride {z=-980}
    physicsBodyDefinitions [{
      boundBone {boneName=dyn_vbuck_mid_B_02}
      boxExtents=<5 5 10>
      constraintSetup {angularLimitsMax=<0 10 0>, angularLimitsMin=<0 -10 -35>, angularTarget=<1 0 0>}
      localJointOffset=<1.5 0 0>
     }]
   }
   n12 = Anim Dynamics - Bone: dyn_vbuck_mid_C_01(ComponentPose=@n11, Alpha=1, ExternalForce="0,0,0") {
    angularDampingOverride=0.998
    angularSpringConstant=100
    bAngularSpring=true
    bOverrideAngularDamping=true
    bUseGravityOverride=true
    boundBone {boneName=dyn_vbuck_mid_C_01}
    componentAppliedLinearAccClamp=<100 100 100>
    componentLinearAccScale=<0.1 0.1 0.1>
    componentLinearVelScale=<0.2 0.2 0.2>
    gravityOverride {z=-700}
    physicsBodyDefinitions [{
      boundBone {boneName=dyn_vbuck_mid_C_01}
      boxExtents=<10 5 10>
      constraintSetup {angularLimitsMax=<0 10 20>, angularLimitsMin=<0 -10 -10>, angularTarget=<1 0 0>}
      localJointOffset=<1 0 0>
     }]
   }
   n13 = Anim Dynamics - Bone: dyn_vbuck_bot_A_01(ComponentPose=@n12, Alpha=1, ExternalForce="0,0,0") {
    angularDampingOverride=0.9997
    angularSpringConstant=100
    bAngularSpring=true
    bOverrideAngularDamping=true
    bUseGravityOverride=true
    boundBone {boneName=dyn_vbuck_bot_A_01}
    componentAppliedLinearAccClamp=<700 500 350>
    componentLinearAccScale=<0.6 0.6 0.6>
    componentLinearVelScale=<0.7 0.7 0.7>
    gravityOverride {z=-980}
    physicsBodyDefinitions [{
      boundBone {boneName=dyn_vbuck_bot_A_01}
      boxExtents=<5 5 10>
      constraintSetup {angularLimitsMax=<0 10 30>, angularLimitsMin=<0 -10 0>, angularTarget=<1 0 0>}
      localJointOffset=<1.5 0 0>
     }]
   }
   n14 = Anim Dynamics - Bone: dyn_vbuck_bot_A_02(ComponentPose=@n13, Alpha=1, ExternalForce="0,0,0") {
    angularDampingOverride=0.9998
    angularSpringConstant=100
    bAngularSpring=true
    bOverrideAngularDamping=true
    bUseGravityOverride=true
    boundBone {boneName=dyn_vbuck_bot_A_02}
    componentAppliedLinearAccClamp=<500 500 350>
    componentLinearAccScale=<0.6 0.6 0.6>
    componentLinearVelScale=<0.7 0.7 0.7>
    gravityOverride {z=-800}
    physicsBodyDefinitions [{
      boundBone {boneName=dyn_vbuck_bot_A_02}
      boxExtents=<10 5 5>
      constraintSetup {angularLimitsMax=<0 10 0>, angularLimitsMin=<0 -10 -50>, angularTarget=<1 0 0>}
      localJointOffset=<1 0 0>
     }]
   }
   n15 = Anim Dynamics - Bone: dyn_vbuck_bot_B_01(ComponentPose=@n14, Alpha=1, ExternalForce="0,0,0") {
    angularDampingOverride=0.999
    angularSpringConstant=100
    bAngularSpring=true
    bOverrideAngularDamping=true
    bUseGravityOverride=true
    boundBone {boneName=dyn_vbuck_bot_B_01}
    componentAppliedLinearAccClamp=<700 500 350>
    componentLinearAccScale=<0.6 0.6 0.6>
    componentLinearVelScale=<0.7 0.7 0.7>
    gravityOverride {z=-980}
    physicsBodyDefinitions [{
      boundBone {boneName=dyn_vbuck_bot_B_01}
      boxExtents=<5 10 10>
      constraintSetup {angularLimitsMax=<0 10 0>, angularLimitsMin=<0 -10 -50>, angularTarget=<1 0 0>}
      localJointOffset=<2 0 0>
     }]
   }
   n16 = Anim Dynamics - Bone: dyn_vbuck_top_A_01(ComponentPose=Local To Component(LocalPose=Input Pose() {name=InPose_0}), Alpha=1, ExternalForce="0,0,0") {
    angularDampingOverride=0.9999
    angularSpringConstant=1500
    bAngularSpring=true
    bOverrideAngularDamping=true
    bUseGravityOverride=true
    boundBone {boneName=dyn_vbuck_top_A_01}
    gravityOverride {z=-800}
    physicsBodyDefinitions [{
      boundBone {boneName=dyn_vbuck_top_A_01}
      boxExtents=<5 10 5>
      constraintSetup {angularLimitsMax=<0 10 35>, angularLimitsMin=<0 -10 -5>, angularTarget=<1 0 0>}
      localJointOffset=<1.2 0 0>
     }]
    simSpaceSettings {simSpaceAngularAlpha=1}
   }
   n17 = Anim Dynamics - Bone: dyn_vbuck_top_A_02(ComponentPose=@n16, Alpha=1, ExternalForce="0,0,0") {
    angularDampingOverride=0.9999
    angularSpringConstant=500
    bAngularSpring=true
    bOverrideAngularDamping=true
    bUseGravityOverride=true
    boundBone {boneName=dyn_vbuck_top_A_02}
    gravityOverride {z=-700}
    physicsBodyDefinitions [{
      boundBone {boneName=dyn_vbuck_top_A_02}
      boxExtents=<5 10 5>
      constraintSetup {angularLimitsMax=<0 10 5>, angularLimitsMin=<0 -10 -50>, angularTarget=<1 0 0>}
      localJointOffset=<2 0 0>
     }]
    simSpaceSettings {maxAngularAcceleration=10, maxAngularVelocity=70, simSpaceAngularAlpha=1}
   }
   n18 = Anim Dynamics - Bone: dyn_vbuck_top_B_01(ComponentPose=@n17, Alpha=1, ExternalForce="0,0,0") {
    angularDampingOverride=0.9999
    angularSpringConstant=1500
    bAngularSpring=true
    bOverrideAngularDamping=true
    bUseGravityOverride=true
    boundBone {boneName=dyn_vbuck_top_B_01}
    gravityOverride {z=-800}
    physicsBodyDefinitions [{
      boundBone {boneName=dyn_vbuck_top_B_01}
      boxExtents=<5 10 5>
      constraintSetup {angularLimitsMax=<0 10 40>, angularLimitsMin=<0 -10 0>, angularTarget=<1 0 0>}
      localJointOffset=<0.5 0 0>
     }]
    simSpaceSettings {simSpaceAngularAlpha=1}
   }
   n19 = Anim Dynamics - Bone: dyn_vbuck_top_C_01(ComponentPose=@n18, Alpha=1, ExternalForce="0,0,0") {
    angularDampingOverride=0.9999
    angularSpringConstant=1500
    bAngularSpring=true
    bOverrideAngularDamping=true
    bUseGravityOverride=true
    boundBone {boneName=dyn_vbuck_top_C_01}
    gravityOverride {z=-800}
    physicsBodyDefinitions [{
      boundBone {boneName=dyn_vbuck_top_C_01}
      boxExtents=<5 10 5>
      constraintSetup {angularLimitsMax=<0 0 40>, angularLimitsMin=<0 0 0>, angularTarget=<1 0 0>}
      localJointOffset=<1.2 0 0>
     }]
    simSpaceSettings {simSpaceAngularAlpha=1}
   }
   n20 = Anim Dynamics - Bone: dyn_vbuck_top_C_02(ComponentPose=@n19, Alpha=1, ExternalForce="0,0,0") {
    angularDampingOverride=0.998
    angularSpringConstant=500
    bAngularSpring=true
    bOverrideAngularDamping=true
    bUseGravityOverride=true
    boundBone {boneName=dyn_vbuck_top_C_02}
    gravityOverride {z=-300}
    physicsBodyDefinitions [{
      boundBone {boneName=dyn_vbuck_top_C_02}
      boxExtents=<10 5 5>
      constraintSetup {angularLimitsMax=<0 10 0>, angularLimitsMin=<0 -10 -50>, angularTarget=<1 0 0>}
      localJointOffset=<2 0 0>
     }]
    simSpaceSettings {maxAngularAcceleration=50, maxAngularVelocity=500, simSpaceAngularAlpha=1}
   }
   n21 = Anim Dynamics - Bone: dyn_vbuck_top_D_01(ComponentPose=@n20, Alpha=1, ExternalForce="0,0,0") {
    angularDampingOverride=0.9
    angularSpringConstant=500
    bAngularSpring=true
    bOverrideAngularDamping=true
    bUseGravityOverride=true
    boundBone {boneName=dyn_vbuck_top_D_01}
    gravityOverride {z=-300}
    physicsBodyDefinitions [{
      boundBone {boneName=dyn_vbuck_top_D_01}
      boxExtents=<10 5 5>
      constraintSetup {angularLimitsMax=<0 10 5>, angularLimitsMin=<0 -10 -50>, angularTarget=<1 0 0>}
      localJointOffset=<2 0 0>
     }]
    simSpaceSettings {maxAngularAcceleration=700, maxAngularVelocity=3500, simSpaceAngularAlpha=1}
   }
   n22 = Anim Dynamics - Bone: dyn_vbuck_top_E_01(ComponentPose=@n21, Alpha=1, ExternalForce="0,0,0") {
    angularDampingOverride=0.9999
    angularSpringConstant=1500
    bAngularSpring=true
    bOverrideAngularDamping=true
    bUseGravityOverride=true
    boundBone {boneName=dyn_vbuck_top_E_01}
    gravityOverride {z=-500}
    physicsBodyDefinitions [{
      boundBone {boneName=dyn_vbuck_top_E_01}
      boxExtents=<10 5 5>
      constraintSetup {angularLimitsMax=<0 5 5>, angularLimitsMin=<0 -5 -50>, angularTarget=<1 0 0>}
      localJointOffset=<2 0 0>
     }]
    simSpaceSettings {maxAngularAcceleration=10, maxAngularVelocity=100, simSpaceAngularAlpha=1}
   }
   n23 = Anim Dynamics - Bone: dyn_vbuck_mid_A_01(ComponentPose=@n22, Alpha=1, ExternalForce="0,0,0") {
    angularDampingOverride=0.9999
    angularSpringConstant=1200
    bAngularSpring=true
    bOverrideAngularDamping=true
    bUseGravityOverride=true
    boundBone {boneName=dyn_vbuck_mid_A_01}
    gravityOverride {z=-800}
    physicsBodyDefinitions [{
      boundBone {boneName=dyn_vbuck_mid_A_01}
      boxExtents=<5 10 5>
      constraintSetup {angularLimitsMax=<0 10 35>, angularLimitsMin=<0 -10 -15>, angularTarget=<1 0 0>}
      localJointOffset=<0.7 0 0>
     }]
    simSpaceSettings {simSpaceAngularAlpha=1}
   }
   n24 = Anim Dynamics - Bone: dyn_vbuck_mid_A_02(ComponentPose=@n23, Alpha=1, ExternalForce="0,0,0") {
    angularDampingOverride=0.9999
    angularSpringConstant=500
    bAngularSpring=true
    bOverrideAngularDamping=true
    bUseGravityOverride=true
    boundBone {boneName=dyn_vbuck_mid_A_02}
    gravityOverride {z=-700}
    physicsBodyDefinitions [{
      boundBone {boneName=dyn_vbuck_mid_A_02}
      boxExtents=<5 10 5>
      constraintSetup {angularLimitsMax=<0 10 0>, angularLimitsMin=<0 -10 -50>, angularTarget=<1 0 0>}
      localJointOffset=<2 0 0>
     }]
    simSpaceSettings {maxAngularAcceleration=10, maxAngularVelocity=70, simSpaceAngularAlpha=1}
   }
   n25 = Anim Dynamics - Bone: dyn_vbuck_mid_B_01(ComponentPose=@n24, Alpha=1, ExternalForce="0,0,0") {
    angularDampingOverride=0.9999
    angularSpringConstant=1500
    bAngularSpring=true
    bOverrideAngularDamping=true
    bUseGravityOverride=true
    boundBone {boneName=dyn_vbuck_mid_B_01}
    gravityOverride {z=-800}
    physicsBodyDefinitions [{
      boundBone {boneName=dyn_vbuck_mid_B_01}
      boxExtents=<5 10 5>
      constraintSetup {angularLimitsMax=<0 10 35>, angularLimitsMin=<0 -10 0>, angularTarget=<1 0 0>}
      localJointOffset=<1.2 0 0>
     }]
    simSpaceSettings {simSpaceAngularAlpha=1}
   }
   n26 = Anim Dynamics - Bone: dyn_vbuck_mid_B_02(ComponentPose=@n25, Alpha=1, ExternalForce="0,0,0") {
    angularDampingOverride=0.998
    angularSpringConstant=500
    bAngularSpring=true
    bOverrideAngularDamping=true
    bUseGravityOverride=true
    boundBone {boneName=dyn_vbuck_mid_B_02}
    gravityOverride {z=-300}
    physicsBodyDefinitions [{
      boundBone {boneName=dyn_vbuck_mid_B_02}
      boxExtents=<10 5 5>
      constraintSetup {angularLimitsMax=<0 10 0>, angularLimitsMin=<0 -10 -35>, angularTarget=<1 0 0>}
      localJointOffset=<2 0 0>
     }]
    simSpaceSettings {maxAngularAcceleration=50, maxAngularVelocity=500, simSpaceAngularAlpha=1}
   }
   n27 = Anim Dynamics - Bone: dyn_vbuck_mid_C_01(ComponentPose=@n26, Alpha=1, ExternalForce="0,0,0") {
    angularDampingOverride=0.9999
    angularSpringConstant=1200
    bAngularSpring=true
    bOverrideAngularDamping=true
    bUseGravityOverride=true
    boundBone {boneName=dyn_vbuck_mid_C_01}
    gravityOverride {z=-800}
    physicsBodyDefinitions [{
      boundBone {boneName=dyn_vbuck_mid_C_01}
      boxExtents=<5 10 5>
      constraintSetup {angularLimitsMax=<0 10 30>, angularLimitsMin=<0 -10 -10>, angularTarget=<1 0 0>}
      localJointOffset=<0.7 0 0>
     }]
    simSpaceSettings {simSpaceAngularAlpha=1}
   }
   n28 = Anim Dynamics - Bone: dyn_vbuck_bot_A_01(ComponentPose=@n27, Alpha=1, ExternalForce="0,0,0") {
    angularDampingOverride=0.9999
    angularSpringConstant=2000
    bAngularSpring=true
    bOverrideAngularDamping=true
    bUseGravityOverride=true
    boundBone {boneName=dyn_vbuck_bot_A_01}
    gravityOverride {z=-700}
    physicsBodyDefinitions [{
      boundBone {boneName=dyn_vbuck_bot_A_01}
      boxExtents=<10 5 5>
      constraintSetup {angularLimitsMax=<0 5 30>, angularLimitsMin=<0 -5 0>, angularTarget=<1 0 0>}
      localJointOffset=<1 0 0>
     }]
    simSpaceSettings {maxAngularAcceleration=700, maxAngularVelocity=3500, simSpaceAngularAlpha=1}
   }
   n29 = Anim Dynamics - Bone: dyn_vbuck_bot_A_02(ComponentPose=@n28, Alpha=1, ExternalForce="0,0,0") {
    angularDampingOverride=0.998
    angularSpringConstant=1000
    bAngularSpring=true
    bOverrideAngularDamping=true
    bUseGravityOverride=true
    boundBone {boneName=dyn_vbuck_bot_A_02}
    gravityOverride {z=-500}
    physicsBodyDefinitions [{
      boundBone {boneName=dyn_vbuck_bot_A_02}
      boxExtents=<10 5 5>
      constraintSetup {angularLimitsMax=<0 5 0>, angularLimitsMin=<0 -5 -40>, angularTarget=<1 0 0>}
      localJointOffset=<1 0 0>
     }]
    simSpaceSettings {maxAngularAcceleration=10, maxAngularVelocity=200, simSpaceAngularAlpha=1}
   }
   n30 = Anim Dynamics - Bone: dyn_vbuck_bot_B_01(ComponentPose=@n29, Alpha=1, ExternalForce="0,0,0") {
    angularDampingOverride=0.9999
    angularSpringConstant=500
    bAngularSpring=true
    bOverrideAngularDamping=true
    bUseGravityOverride=true
    boundBone {boneName=dyn_vbuck_bot_B_01}
    gravityOverride {z=-800}
    physicsBodyDefinitions [{
      boundBone {boneName=dyn_vbuck_bot_B_01}
      boxExtents=<10 5 5>
      constraintSetup {angularLimitsMax=<0 5 0>, angularLimitsMin=<0 -5 -40>, angularTarget=<1 0 0>}
      localJointOffset=<0.7 0 0>
     }]
    simSpaceSettings {maxAngularAcceleration=100, maxAngularVelocity=500, simSpaceAngularAlpha=1}
   }
   Output Pose(Result=Blend Poses by bool(BlendPose_0=Component To Local(ComponentPose=@n30), BlendPose_1=Component To Local(ComponentPose=@n15), bActiveValue=Get bIsInFrontEnd())) {layerGroup=None}
  ]}
 EventGraph {nodes [Event BlueprintUpdateAnimation()]}
]
variables []
